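# Physics asset for the Troops pickaxe, a melee weapon mesh.
# It holds a single rigid body on the b_AR_Root bone, shaped as one rotated capsule offset from the bone origin.
# There are no constraints, so the pickaxe simulates and collides as one solid piece.
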PhysicsAsset /Game/Weapons/FORT_Melee/Pickaxe_Troops/Meshes/Troops_Pickaxe_PhysicsAsset
bodies {b_AR_Root [{shapeType=Capsule, length=36.2008, radius=12.2427, center=<-9.2355 -1.0902 -22.3975>, rotation=<rot 8.3359 -90 18.2925>}]}
constraints []
