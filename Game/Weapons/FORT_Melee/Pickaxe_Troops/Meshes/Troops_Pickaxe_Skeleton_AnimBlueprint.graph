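# Animation Blueprint for the Troops pickaxe skeleton that adjusts the pickaxe's b_AR_Root bone to suit the wearer's body type and gender.
# When character and pickaxe customization loads, one custom part's permitted body type and gender set four booleans: Large or not, male or female.
# Each boolean enables one of four chained additive Transform Bone nodes on the input pose, with per-build offset and scale and a shared pitch.
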AnimBlueprint /Game/Weapons/FORT_Melee/Pickaxe_Troops/Meshes/Troops_Pickaxe_Skeleton_AnimBlueprint
target_skeleton=&/Troops_Pickaxe_Skeleton
graphs [
 AnimGraph {nodes [
   n1 = "Transform (Modify) Bone - Bone: b_AR_Root"!(
    ComponentPose=Local To Component(LocalPose=Input Pose())
    bAlphaBoolEnabled:bool=Get not Large F() -> (not Large F:bool)
    Translation="0,0,15"
    Rotation="-75,0,0"
    Scale="1,1,1"
   ) -> (Pose) {
    alphaInputType=Bool
    rotationMode=BMM_Additive
    scaleMode=BMM_Additive
    scaleSpace=BCS_BoneSpace
    translationMode=BMM_Additive
    translationSpace=BCS_BoneSpace
    alphaBoolBlend {blendOption=Linear}
    boneToModify {boneName=b_AR_Root}
    rotation {pitch=-75}
    translation {z=15}
   }
   n2 = "Transform (Modify) Bone - Bone: b_AR_Root"(ComponentPose=@n1, bAlphaBoolEnabled=Get not Large M() -> (not Large M:bool), Translation="0,-0.5,15", Rotation="-75,0,0", Scale="1,1,1") {
    alphaInputType=Bool
    rotationMode=BMM_Additive
    scaleMode=BMM_Additive
    scaleSpace=BCS_BoneSpace
    translationMode=BMM_Additive
    translationSpace=BCS_BoneSpace
    alphaBoolBlend {blendOption=Linear}
    boneToModify {boneName=b_AR_Root}
    rotation {pitch=-75}
    translation {y=-0.5, z=15}
   }
   n3 = "Transform (Modify) Bone - Bone: b_AR_Root"(ComponentPose=@n2, bAlphaBoolEnabled=Get Large F() -> (Large F:bool), Translation="-1,-0.5,15", Rotation="-75,0,0", Scale="1.15,1.15,1.15") {
    alphaInputType=Bool
    rotationMode=BMM_Additive
    scaleMode=BMM_Additive
    scaleSpace=BCS_BoneSpace
    translationMode=BMM_Additive
    translationSpace=BCS_BoneSpace
    alphaBoolBlend {blendOption=Linear}
    boneToModify {boneName=b_AR_Root}
    rotation {pitch=-75}
    scale=<1.15 1.15 1.15>
    translation=<-1 -0.5 15>
   }
   n4 = "Transform (Modify) Bone - Bone: b_AR_Root"(ComponentPose=@n3, bAlphaBoolEnabled=Get Large M() -> (Large M:bool), Translation="0,-1.5,17", Rotation="-75,0,0", Scale="1.3,1.3,1.3") {
    alphaInputType=Bool
    rotationMode=BMM_Additive
    rotationSpace=BCS_ParentBoneSpace
    scaleMode=BMM_Additive
    scaleSpace=BCS_BoneSpace
    translationMode=BMM_Additive
    translationSpace=BCS_BoneSpace
    alphaBoolBlend {blendOption=Linear}
    boneToModify {boneName=b_AR_Root}
    rotation {pitch=-75}
    scale=<1.3 1.3 1.3>
    translation {y=-1.5, z=17}
   }
   Output Pose(Result=Component To Local(ComponentPose=@n4)) {layerGroup=None}
  ]}
 EventGraph {nodes [
   n1 = Event OnCharacterAndPickaxeCustomizationLoaded!() -> (
    OutputDelegate
    then
    CustomBaseMesh:Skeletal Mesh Component Object Reference
    CustomParts:Array of Custom Character Part Object References
    CustomMeshParts:Array of Skeletal Mesh Component Budgeted Object References
   ) {is=title K2Node_Event}
   n2 = Get Array Item with EFort Custom Part Type(Array:Array of Custom Character Part Object References=@n1.CustomParts) -> (Output:Custom Character Part Object Reference)
   n3 = "Equal (Enum)"(
    A:EFortCustomBodyType Enum=Get BodyTypesPermitted!(self:Custom Character Part Object Reference=@n2) -> (BodyTypesPermitted:EFortCustomBodyType Enum=NONE)
    B:EFortCustomBodyType Enum=Large
   )
   n4 = "Equal (Enum)"(
    A:EFortCustomGender Enum=Get GenderPermitted!(self:Custom Character Part Object Reference=@n2) -> (GenderPermitted:EFortCustomGender Enum=Invalid)
    B:EFortCustomGender Enum=Male
   )
   > Set not Large M(not Large M:bool=AND Boolean(A=NOT Boolean(A=@n3), B=@n4)) -> (Output_Get:bool)
   n5 = "Equal (Enum)"(A=Get GenderPermitted(self=@n2) -> (GenderPermitted=Invalid), B=Female)
   > Set not Large F(not Large F:bool=AND Boolean(A=NOT Boolean(A=@n3), B=@n5)) -> (Output_Get:bool)
   > Set Large M(Large M:bool=AND Boolean(A=@n3, B=@n4)) -> (Output_Get:bool)
   > Set Large F(Large F:bool=AND Boolean(A=@n3, B=@n5)) -> (Output_Get:bool)
  ]}
]
variables [
 Large F {type=boolean}
 Large M {type=boolean}
 not Large F {type=boolean}
 not Large M {type=boolean}
]
